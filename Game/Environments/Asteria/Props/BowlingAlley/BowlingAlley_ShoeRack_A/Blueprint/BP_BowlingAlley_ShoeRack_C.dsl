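; Shoe rack prop Blueprint for the Asteria bowling alley that customizes how its parent class places instanced static meshes.
; The AddNewInstanceToComponent override turns each requested instance into two parent calls, offset along X in opposite directions by the AddonDeco OffsetDistance.
; The negative-side copy mirrors the incoming scale on X, while the positive-side copy uses default scale, and neither keeps the incoming rotation.
; The construction script only calls the parent.
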(event EventBeginPlay
  (|CallParentFunction) ; This node is disabled and will not be called.\nDrag off pins to build functionality.
)

(event Collision|EventActorBeginOverlap
  (params (OtherActor :type Actor%20Object%20Reference))
  (|CallParentFunction OtherActor) ; This node is disabled and will not be called.\nDrag off pins to build functionality.
)

(event EventTick
  (params (DeltaSeconds :type Float%20%28single-precision%29))
  (|CallParentFunction DeltaSeconds) ; This node is disabled and will not be called.\nDrag off pins to build functionality.
)

(fn AddNewInstanceToComponent
  (params (comp :type Instanced%20Static%20Mesh%20Component%20Object%20Reference) (InTransform :type Transform%20%28by%20ref%29))
  (LocalVariables|Setl_comp comp)
  (LocalVariables|Setl_transform InTransform)
  (Utilities|FlowControl|Sequence
    (:then_0
      (|CallParentFunction (LocalVariables|Getl_comp) (Math|Transform|MakeTransform (Math|Vector|MakeVector (- (.x (.location (LocalVariables|Getl_transform))) (Variables|AddonDeco|GetOffsetDistance)) (.y (.location (LocalVariables|Getl_transform))) (.z (.location (LocalVariables|Getl_transform)))) "0, 0, 0" (* (.scale (LocalVariables|Getl_transform)) "-1.000000,1.000000,1.000000"))))
    (:then_1
      (|CallParentFunction (LocalVariables|Getl_comp) (Math|Transform|MakeTransform (Math|Vector|MakeVector (+ (.x (.location (LocalVariables|Getl_transform))) (Variables|AddonDeco|GetOffsetDistance)) (.y (.location (LocalVariables|Getl_transform))) (.z (.location (LocalVariables|Getl_transform)))))))))

(fn ConstructionScript
  (|CallParentFunction))
